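# Editor utility for building actors, adding a Check Draw Distances action that reviews the actors selected in the editor.
# The action passes the selection and a local list of soft review classes to the shared Run Reviews on Objects routine.
# The resulting per-object review map goes to Display Review Results, with a local placeholder map in the asset slot.
# The draw distance checks themselves live outside this graph.
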EditorUtilityBlueprint /Game/NeverCook/EditorUtilities/ActionUtilities/ActorActions_BuildingActors
graphs [
 Check Draw Distances {nodes [
   Check Draw Distances() {is=title K2Node_FunctionEntry}
   > n1 = GetSelectedActors()
   n2 = Self-Reference()
   > n3 = Run Reviews on Objects!(
    execute
    Objects:Array of Object References=@n1
    Reviews:Array of Soft Class References=Get L_Reviews() -> (L_Reviews:Array of Object Soft Class References)
    __WorldContext:Object Reference=@n2
   ) -> (then, ObjectReviewMap:Map of Object References to Structures)
   > Display Review Results!(
    execute
    AssetMap:Map of Asset Data Structures to Structures=Get l_assetsBlank() -> (l_assetsBlank:Map of Asset Data Structures to Structures)
    ObjectMap:Map of Object References to Structures=@n3
    __WorldContext:Object Reference=@n2
   ) -> (then)
  ]}
 EventGraph {links [], nodes []}
]
variables []
